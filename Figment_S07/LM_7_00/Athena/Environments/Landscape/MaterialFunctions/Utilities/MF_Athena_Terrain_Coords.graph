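# Material function that converts world position into normalized, terrain-wide XY coordinates, with its parameters grouped under TerrainColorMatch.
# It subtracts TerrainMinCorner from the XY of the WorldPosition input, which defaults to absolute world position, then divides by TerrainSize.
# The saturated result is output directly and as WithInterpolator, which routes the math through a VertexInterpolator to compute it per vertex.
# A leftover reciprocal-size chain feeds neither output.
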MaterialFunction /Figment_S07/LM_7_00/Athena/Environments/Landscape/MaterialFunctions/Utilities/MF_Athena_Terrain_Coords
graph {nodes [
  n1 = ScalarParameter() {defaultValue=306200.7812, group=TerrainColorMatch, parameterName=TerrainSize}
  n2 = Divide(A=@n1)
  n3 = Multiply(A=@n2) {constB=-1}
  AppendVector(A=@n3, B=@n3)
  n4 = FunctionInput(Preview=WorldPosition().XYZ) {bUsePreviewValueAsDefault=true, inputName=WorldPosition}
  n5 = Divide(
   A=Subtract(
    A=ComponentMask(None=@n4) {g=true, r=true}
    B=ScalarParameter() {defaultValue=-153000, group=TerrainColorMatch, parameterName=TerrainMinCorner}
   )
   B=@n1
  )
  FunctionOutput(None=Saturate(None=@n5))
  FunctionOutput(None=Saturate(None=VertexInterpolator(VS=@n5).PS)) {outputName=WithInterpolator}
  Multiply(A=@n2)
 ]}
