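# Simplified arid desert road landscape material, usable on spline meshes, built on material attributes.
# The arid sand function's simple output gets a low specular override, then optionally passes through the desert road function.
# An optional topo switch multiplies base color by a terrain colormap, sampled at terrain coordinates and scaled by the landscape parameter collection.
# A legacy snow function follows, taking an inverted mask, apparently road alpha, as extra snow mask.
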Material /Figment_S07/LM_7_00/Athena/Environments/Landscape/Material/Arid/M_AthenaDesertRoad_Simple
graph {
 properties {bUseMaterialAttributes=true, bUsedWithSplineMeshes=true, pixelDepthOffsetMode=PDOM_Legacy, usedShadingModels=Default Lit}
 nodes [
  n1 = Abs(None=CollectionParameter() {collection=/Game/Athena/Environments/Landscape/MPC/MPC_Landscape_S07, parameterName=Arid_Sand_TexSize})
  ComponentMask(None=Divide(A=WorldPosition() {worldPositionShaderOffset=WPT_ExcludeAllShaderOffsets}.XYZ, B=Multiply(A=@n1) {constB=-1})) {g=true, r=true}
  n2 = MaterialFunctionCall!(In, UseRoughness=StaticBool()) {materialFunction=/Figment_S07/LM_7_00/Athena/Environments/Landscape/MaterialFunctions/Arid/MF_Athena_Arid_Sand}
  n3 = SetMaterialAttributes!(MaterialAttributes=@n2.Simple_Material, Specular=Constant() {desc=Landscape Arid override spec value, r=0.1}) {attributeSetTypes [#1]}
  n4 = MaterialFunctionCall!(In=StaticSwitch(True=@n3, False=@n2.Simple_Material)) {materialFunction=/Figment_S07/LM_7_00/Athena/Environments/Landscape/MaterialFunctions/Arid/MF_Athena_DesertRoad}
  n5 = StaticSwitchParameter(True=@n4.Result, False=@n3) {defaultValue=true, parameterName=UseDesertRoad}
  n6 = GetMaterialAttributes(None=@n5) {attributeGetTypes [#2, #1]}
  n7 = TextureSample(UVs=MaterialFunctionCall!(WorldPosition) {materialFunction=/Figment_S07/LM_7_00/Athena/Environments/Landscape/MaterialFunctions/Utilities/MF_Athena_Terrain_Coords}.Result) {
   samplerSource=SSM_Wrap_WorldGroupSettings
   samplerType=SAMPLERTYPE_LinearColor
   texture=/Figment_S07/LM_7_00/Athena/Environments/Landscape/Textures/T_Athena_Terrain_CombinedColors_01
  }
  n8 = Multiply(
   A=@n6.Base Color
   B=Multiply(A=@n7.RGB, B=CollectionParameter() {collection=&MPC_Landscape_S07, parameterName=ColormapDivisor})
  )
  MaterialFunctionCall!(
   ExtraSnowMask=OneMinus(None=NamedRerouteUsage())
   Simple Mask=StaticBool()
   ForceUseInWSNormal
   UseActorPosition
   UseSnowMaskNormal
   UseColorTest
   UseColorTestRock
   In=StaticSwitchParameter(
    True=SetMaterialAttributes!(MaterialAttributes=@n6.MaterialAttributes, Base Color=@n8) {attributeSetTypes [#2]}
    False=@n5
   ) {defaultValue=true, parameterName=UseTopo}
  ) {materialFunction=/Game/Packages/Fortnite_UniqueMaterials_1/MaterialFunctions/Trees/MF_Tree_Snow_Legacy}
  MaterialFunctionCall!(Material=@n5) {materialFunction=/Engine/Functions/MaterialLayerFunctions/MatLayerBlend_BreakBaseColor}
  NamedRerouteDeclaration(None=@n4.Alpha) {name=RoadAlpha, nodeColor=<rgba 0 1 0.0391 1>}
  VertexInterpolator(VS=VertexColor().A)
 ]
}
